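# Unlit additive material for taillight trails on the Sourdough vehicle body, built for Niagara ribbons.
# Scrolling grayscale noise is masked by a power falloff along U, fading away from the U=0 end.
# The color chain adds a tighter hot spot at that end, tints it by particle color and boosts brightness.
# The alpha chain adds a V-centered parabolic gradient to the noise and scales it by particle alpha.
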Material /VehicleCosmetics/Bodies/Sourdough/FX/M_TaillightTrails
graph {
 properties {bUsedWithNiagaraRibbons=true, blendMode=BLEND_Additive, pixelDepthOffsetMode=PDOM_Legacy, shadingModel=MSM_Unlit, usedShadingModels=Unlit}
 nodes [
  n1 = ParticleColor()
  n2 = TextureCoordinate()
  n3 = ComponentMask(None=@n2) {g=true}
  n4 = OneMinus(None=Clamp(None=@n3))
  n5 = TextureCoordinate()
  n6 = OneMinus(None=Clamp(None=ComponentMask(None=@n2) {r=true}))
  n7 = Multiply(
   A=TextureSample(UVs=Panner(Coordinate=@n5) {speedX=-1}) {samplerType=SAMPLERTYPE_LinearGrayscale, texture=&/T_FX_TaillightNoise00_M}.RGB
   B=Power(Base=@n6) {constExponent=5}
  )
  Multiply(A=@n1.A, B=Add(A=Multiply(A=@n4, B=@n3), B=@n7))
  n8 = Clamp(None=Power(Base=@n6) {constExponent=24})
  Multiply(A=Multiply(A=@n1.RGB, B=Add(A=@n7, B=@n8)) {constA=0.1}) {constB=16}
 ]
}
