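# Niagara system for the Sourdough vehicle body that draws short red streak trails from the taillights.
# A high-rate emitter spawns short-lived particles with curve-scaled velocity along X, rendered through sprite and ribbon renderers.
# Curves shape alpha and ribbon width over life, while InheritSourceMovement applies a decaying share of source motion to particle positions.
# A second, ribbon-only copy with fainter alpha is disabled.
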NiagaraSystem /VehicleCosmetics/Bodies/Sourdough/FX/NS_Sourdough_TaillightTrail
dependencies {usedDataInterfaces [&NiagaraDataInterfaceCurve], usedDynamicInputs [], usedModules [&SystemState]}
emitters {
 TaillightTrail {
  emitterUpdateScript [
   &EmitterState(Life Cycle Mode=::NewEnumerator1, Loop Duration=1, UseLoopDelay=false)
   &SpawnRate(SpawnRate=2048, Use Spawn Probability=false)
  ]
  particleSpawnScript [
   &InitializeParticle(
    Color=<rgba 1 0 0 1>
    Lifetime=0.08
    Mass Mode=::NewEnumerator1
    Ribbon Facing Mode=::NewEnumerator1
    Ribbon Twist Mode:ENiagara_UnsetDirectSet=::NewEnumerator1
    Ribbon Width=6
    Ribbon Width Mode=::NewEnumerator1
    Sprite Size Mode=::NewEnumerator3
    Uniform Sprite Size=1
    UsePositionOffset=false
    Write Material Random=true
   )
   &AddVelocity(
    Velocity=<800 0 0>
    Velocity Speed Scale=&FloatFromCurve(FloatCurve=DataInterface() {propertyValues {Curve {keys [
         {interpMode=RCIM_Cubic, tangentMode=RCTM_User, time=0, value=1}
         {arriveTangent=-0.1901, interpMode=RCIM_Cubic, leaveTangent=-0.1901, tangentMode=RCTM_User, time=1, value=0.15}
         {interpMode=RCIM_Cubic, tangentMode=RCTM_User, time=1, value=0}
        ]}}})
   )
  ]
  particleUpdateScript [
   &ParticleState()
   &ScaleColor(
    Scale Alpha=&FloatFromCurve(
     FloatCurve=DataInterface() {propertyValues {Curve {keys [<at 0 0 RCIM_Cubic>, {arriveTangent=0.002, interpMode=RCIM_Cubic, leaveTangent=0.002, time=0.05, value=1}, <at 1.0011 0.002 RCIM_Cubic>]}}}
     Scale Curve=0.2
    )
    ScaleA=true
    ScaleRGB=true
   )
   &InheritSourceMovement(
    Applied Position Scale=&Multiply_VectorByFloat(
     Vector=<1 1 1>
     Float=&FloatFromCurve(FloatCurve=DataInterface() {propertyValues {Curve {keys [{arriveTangent=-9.6986, interpMode=RCIM_Cubic, time=0, value=0.75}, {interpMode=RCIM_Cubic, leaveTangent=-10.0133, time=1, value=0}]}}})
    )
    ApplyForce=false
    ApplyPosition=true
    ApplyPositionFalloff=false
   )
   &SolveForcesAndVelocity(Acceleration Limit=16, Clamp Velocity=false, Limit Acceleration=false, Speed Limit=16)
   &ScaleRibbonWidth(Ribbon Width Scale=&FloatFromCurve(
     FloatCurve=DataInterface() {propertyValues {Curve {keys [
         {arriveTangent=-0.2138, interpMode=RCIM_Cubic, leaveTangent=-0.2138, tangentMode=RCTM_User, time=0, value=0.99}
         {arriveTangent=-0.49, interpMode=RCIM_Cubic, leaveTangent=-0.49, time=0.4, value=0.875}
         {arriveTangent=-0.8168, interpMode=RCIM_Cubic, leaveTangent=-0.8168, tangentMode=RCTM_User, time=1, value=0.5}
        ]}}}
     Scale Curve=4
    ))
  ]
  rendererClasses [&NiagaraSpriteRendererProperties, &NiagaraRibbonRendererProperties]
  renderers [
   {rendererClass=&NiagaraSpriteRendererProperties}
   {rendererClass=&NiagaraRibbonRendererProperties, rendererIndex=1}
  ]
 }
 TaillightTrail001 {
  bEnabled=false
  emitterUpdateScript [
   &EmitterState(Life Cycle Mode=::NewEnumerator1, Loop Duration=1, UseLoopDelay=false)
   &SpawnRate(SpawnRate=2048, Use Spawn Probability=false)
  ]
  particleSpawnScript [
   &InitializeParticle(
    Color=<rgba 1 0 0 1>
    Lifetime=0.08
    Mass Mode=::NewEnumerator1
    Ribbon Facing Mode=::NewEnumerator1
    Ribbon Twist Mode:ENiagara_UnsetDirectSet=::NewEnumerator1
    Ribbon Width=6
    Ribbon Width Mode=::NewEnumerator1
    Sprite Size Mode=::NewEnumerator3
    Uniform Sprite Size=1
    UsePositionOffset=false
    Write Material Random=true
   )
   &AddVelocity(
    Velocity=<800 0 0>
    Velocity Speed Scale=&FloatFromCurve(FloatCurve=DataInterface() {propertyValues {Curve {keys [
         {interpMode=RCIM_Cubic, tangentMode=RCTM_User, time=0, value=1}
         {arriveTangent=-0.1901, interpMode=RCIM_Cubic, leaveTangent=-0.1901, tangentMode=RCTM_User, time=1, value=0.15}
         {interpMode=RCIM_Cubic, tangentMode=RCTM_User, time=1, value=0}
        ]}}})
   )
  ]
  particleUpdateScript [
   &ParticleState()
   &ScaleColor(
    Scale Alpha=&FloatFromCurve(
     FloatCurve=DataInterface() {propertyValues {Curve {keys [<at 0 0 RCIM_Cubic>, {arriveTangent=0.002, interpMode=RCIM_Cubic, leaveTangent=0.002, time=0.05, value=1}, <at 1.0011 0.002 RCIM_Cubic>]}}}
     Scale Curve=0.05
    )
    ScaleA=true
    ScaleRGB=true
   )
   &InheritSourceMovement(
    Applied Position Scale=&Multiply_VectorByFloat(
     Vector=<1 1 1>
     Float=&FloatFromCurve(FloatCurve=DataInterface() {propertyValues {Curve {keys [{arriveTangent=-9.6986, interpMode=RCIM_Cubic, time=0, value=0.75}, {interpMode=RCIM_Cubic, leaveTangent=-10.0133, time=1, value=0}]}}})
    )
    ApplyForce=false
    ApplyPosition=true
    ApplyPositionFalloff=false
   )
   &SolveForcesAndVelocity(Acceleration Limit=16, Clamp Velocity=false, Limit Acceleration=false, Speed Limit=16)
   &ScaleRibbonWidth(Ribbon Width Scale=&FloatFromCurve(
     FloatCurve=DataInterface() {propertyValues {Curve {keys [
         {arriveTangent=-0.2138, interpMode=RCIM_Cubic, leaveTangent=-0.2138, tangentMode=RCTM_User, time=0, value=0.99}
         {arriveTangent=-0.49, interpMode=RCIM_Cubic, leaveTangent=-0.49, time=0.4, value=0.875}
         {arriveTangent=-0.8168, interpMode=RCIM_Cubic, leaveTangent=-0.8168, tangentMode=RCTM_User, time=1, value=0.5}
        ]}}}
     Scale Curve=4
    ))
  ]
  rendererClasses [&NiagaraRibbonRendererProperties]
  renderers [{rendererClass=&NiagaraRibbonRendererProperties}]
 }
}
summary {
 systemName=NS_Sourdough_TaillightTrail
 emitters [
  {
   emitterName=TaillightTrail
   rendererClasses [&NiagaraSpriteRendererProperties, &NiagaraRibbonRendererProperties]
  }
  {bEnabled=false, emitterName=TaillightTrail001, rendererClasses [&NiagaraRibbonRendererProperties]}
 ]
 userVariables []
}
