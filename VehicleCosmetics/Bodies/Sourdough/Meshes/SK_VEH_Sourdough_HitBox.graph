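# Physics asset that supplies a simplified hit box for the Sourdough vehicle body skeletal mesh.
# A single body on the root bone holds one box shape, offset forward along X and raised above the bone origin.
# It has no other bodies or constraints, so collision is one rigid volume with no per-part detail.
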PhysicsAsset /VehicleCosmetics/Bodies/Sourdough/Meshes/SK_VEH_Sourdough_HitBox
bodies {root [{shapeType=Box, extentX=406.4614, extentY=263.0012, extentZ=109.309, shapeName=root_box, center=<65 0 96>}]}
constraints []
